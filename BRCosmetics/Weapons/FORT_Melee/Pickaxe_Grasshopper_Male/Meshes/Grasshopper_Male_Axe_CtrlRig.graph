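# Control Rig for the Grasshopper Male pickaxe that post-adjusts the local motion of its front and back "dyn_mallet" bones.
# The front mallet keeps its current Y translation while X and Z lock to their initial values, restricting it to sliding on one axis.
# The back mallet gets the same lock, but its Y offset from initial is inverted and slightly amplified.
# Both bones are written with identity rotation and unit scale.
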ControlRigBlueprint /BRCosmetics/Weapons/FORT_Melee/Pickaxe_Grasshopper_Male/Meshes/Grasshopper_Male_Axe_CtrlRig
graphs [
 RigVMFunctionLibrary {links [], nodes []}
 RigVMModel {nodes [
   RigUnit_BeginExecution = &RigUnit_BeginExecution()
   RigUnit_GetTransform = &RigUnit_GetTransform(Item=(Type=Bone,Name="dyn_mallet_fr"), Space=LocalSpace, bInitial=true)
   RigUnit_SetTransform = &RigUnit_SetTransform(
    Item=(Type=Bone,Name="dyn_mallet_fr")
    Space=LocalSpace
    Value=(Rotation=(X=0,Y=0,Z=0,W=1),Translation=(X=0.692,Y=-12.5287,Z=41.5769),Scale3D=(X=1,Y=1,Z=1))
    Weight=1
    ExecutePin=@RigUnit_BeginExecution.ExecutePin
    Value.Translation.X=@RigUnit_GetTransform.Transform.Translation.X
    Value.Translation.Z=@RigUnit_GetTransform.Transform.Translation.Z
    Value.Translation.Y=&RigUnit_GetTransform(Item=(Type=Bone,Name="dyn_mallet_fr"), Space=LocalSpace, bInitial=false).Transform.Translation.Y
   )
   n1 = &RigUnit_GetTransform(Item=(Type=Bone,Name="dyn_mallet_bk"), Space=LocalSpace, bInitial=true)
   n2 = /Script/RigVM.RigVMFunction_MathFloatSub(
    A=&RigUnit_GetTransform(Item=(Type=Bone,Name="dyn_mallet_bk"), Space=LocalSpace, bInitial=false).Transform.Translation.Y
    B=@n1.Transform.Translation.Y
   ) {is=struct RigVMUnitNode, template="Subtract::Execute(in A,in B,out Result)"}
   > &RigUnit_SetTransform(
    Item=(Type=Bone,Name="dyn_mallet_bk")
    Space=LocalSpace
    Value=(Rotation=(X=0,Y=0,Z=0,W=1),Translation=(X=0.692,Y=-12.5287,Z=41.5769),Scale3D=(X=1,Y=1,Z=1))
    Weight=1
    Value.Translation.Y=&RigVMFunction_MathFloatAdd(
     A=&RigVMFunction_MathFloatMul(B=-1.25, A=@n2.Result) {template="Multiply::Execute(in A,in B,out Result)"}.Result
     B=@n1.Transform.Translation.Y
    ) {template="Add::Execute(in A,in B,out Result)"}.Result
    Value.Translation.X=@n1.Transform.Translation.X
    Value.Translation.Z=@n1.Transform.Translation.Z
   )
  ]}
]
hierarchy [
 b_AR_Root {loc=<0 0 0>, rot=<0 0 0.7071 0.7071>}
 Melee_TwoHanded_Handle {parent=b_AR_Root, loc=<8.1477 -0.008 -29.0793>, rot=<0 0 0.7071 -0.7071>}
 dyn_mallet_fr {parent=Melee_TwoHanded_Handle, loc=<0.684 -20.6764 70.6562>, rot=<0 0 0 -1>}
 dyn_mallet_bk {parent=Melee_TwoHanded_Handle, loc=<0.6886 18.2076 70.6562>, rot=<0 0 0 -1>}
]
